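; BP_MeshNetwork_Speaker is a prototype Blueprint, named as a speaker within a mesh network system.
; It declares BeginPlay and EndPlay events and a construction script, all with empty bodies.
; The graphs shown contain no logic, so any speaker or networking behavior is not defined in them.
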(event EventBeginPlay)

(event EventEndPlay
  (params (EndPlayReason :type EEndPlayReason%20Enum :default "Destroyed")))

(fn ConstructionScript)
